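# Sound cue for the end of searching a fireplace loot container, spatialized with the prop-search attenuation on the player foley class.
# It mixes the shared "no more loot" UI sound with two foley layers, each picking randomly from three fireplace or borrowed firepit search waves.
# Both foley layers fade out quickly through volume envelopes and get slight pitch and volume modulation.
# A concurrency override limits overlapping instances, stopping the farthest first.
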SoundCue /Figment_S09/Sounds/Foley_Loot/Containers/Fireplace/Fireplace_SearchEnd_Cue
root=Mixer_3
nodes [
 {
  id=Enveloper_27
  children [Random_11]
  properties {pitchCurve {editorCurveData {keys [{time=0}]}}, volumeCurve {editorCurveData {keys [<at 0 1>,<at 0.6 0>]}}}
 }
 {
  id=Enveloper_28
  children [Random_12]
  properties {pitchCurve {editorCurveData {keys [{time=0}]}}, volumeCurve {editorCurveData {keys [<at 0 1>,<at 0.6 0>]}}}
 }
 {id=Mixer_3, children [WavePlayer_218, Modulator_8, Modulator_9], properties {inputVolume [0.8,1,0.7]}}
 {id=Modulator_8, children [Enveloper_27], properties {pitchMax=1.07, pitchMin=0.92, volumeMax=1.07, volumeMin=0.92}}
 {id=Modulator_9, children [Enveloper_28], properties {pitchMax=1.07, pitchMin=0.92, volumeMax=1.07, volumeMin=0.92}}
 {id=Random_11, children [WavePlayer_30, WavePlayer_31, WavePlayer_32], properties {weights [1,1,1]}}
 {id=Random_12, children [WavePlayer_33, WavePlayer_34, WavePlayer_35], properties {weights [1,1,1]}}
 {id=WavePlayer_218, properties {soundWaveAssetPtr=/Game/Sounds/Fort_UI_Sounds/Gameplay/Fort_UI_NoMoreLoot}}
 {id=WavePlayer_30, properties {soundWaveAssetPtr=&/Fireplace_Search01}}
 {id=WavePlayer_31, properties {soundWaveAssetPtr=&/Fireplace_Search02}}
 {id=WavePlayer_32, properties {soundWaveAssetPtr=&/Fireplace_Search08}}
 {id=WavePlayer_33, properties {soundWaveAssetPtr=/Figment_S09/Sounds/Foley_Loot/Containers/Firepit/FirePit_Search13}}
 {id=WavePlayer_34, properties {soundWaveAssetPtr=/Figment_S09/Sounds/Foley_Loot/Containers/Firepit/FirePit_Search07}}
 {id=WavePlayer_35, properties {soundWaveAssetPtr=/Figment_S09/Sounds/Foley_Loot/Containers/Firepit/FirePit_Search12}}
]
properties {
 attenuationSettings=/Game/Sounds/Attenuation/Fort_Prop_Search_Att
 bOverrideConcurrency=true
 maxDistance=1300
 soundClassObject=&Fort_Player_Foley
 volumeMultiplier=2.8
 concurrencyOverrides {maxCount=4, resolutionRule=StopFarthestThenPreventNew}
}
